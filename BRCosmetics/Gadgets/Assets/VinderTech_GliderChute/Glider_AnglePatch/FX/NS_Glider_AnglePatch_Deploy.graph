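# Niagara effect for deploying the AnglePatch glider: a mostly green burst of flashes, glows, goo sprites, spiraling mesh trails and crystal debris.
# Shared system values for delay, fade-in window, offset and progress drive the glow, goo and trail emitters.
# A hand position, apparently sampled from the pawn skeleton, anchors flares and two partitioned ribbons that curl into helices.
# VariantPicker shifts flare and crystal tint, while front-end modifiers adjust placement and scale.
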NiagaraSystem /BRCosmetics/Gadgets/Assets/VinderTech_GliderChute/Glider_AnglePatch/FX/NS_Glider_AnglePatch_Deploy
dependencies {
 usedDataInterfaces [
  &NiagaraDataInterfaceColorCurve
  &NiagaraDataInterfaceCurve
  &NiagaraDataInterfaceArrayFloat
  &NiagaraDataInterfaceCamera
  &NiagaraDataInterfaceVectorCurve
  &NiagaraDataInterfaceSpriteRendererInfo
 ]
 usedDynamicInputs [
  &MaskIntBySpawnGroup
  &MakeCustomLinearColorFromBool
  &Subtract_Float
  &TransformPosition
  &AddVectorToPosition
 ]
 usedModules [
  &self:SystemSpawnScript.SystemScriptSource.SystemScriptGraph.NiagaraNodeAssignment_2.SetVariables_#1
  &self:SystemSpawnScript.SystemScriptSource.SystemScriptGraph.NiagaraNodeAssignment_3.SetVariables_#2
  &self:SystemSpawnScript.SystemScriptSource.SystemScriptGraph.NiagaraNodeAssignment_4.SetVariables_#3
  &self:SystemSpawnScript.SystemScriptSource.SystemScriptGraph.NiagaraNodeAssignment_6.SetVariables_#4
  &self:SystemSpawnScript.SystemScriptSource.SystemScriptGraph.NiagaraNodeAssignment_5.SetVariables_#5
  &SystemState
  &self:SystemSpawnScript.SystemScriptSource.SystemScriptGraph.NiagaraNodeAssignment_1.SetVariables_#6
  &self:SystemSpawnScript.SystemScriptSource.SystemScriptGraph.NiagaraNodeAssignment_0.SetVariables_#7
  &SampleSkeletalMesh
  &self:SystemSpawnScript.SystemScriptSource.SystemScriptGraph.NiagaraNodeAssignment_7.SetVariables_#8
 ]
}
emitters {
 NE_CrystalDebris {
  emitterUpdateScript [
   &EmitterState()
   &SpawnBurst_Instantaneous(
    Spawn Count=&RandomRangeInt(Minimum=13, Maximum=20)
    Spawn Time=$System.SpawnDelay
    Use Spawn Probability=false
    UseLoopCountLimit=false
   )
  ]
  particleSpawnScript [
   /SetVariables(Particles.CrystalSizeAlpha:NiagaraFloat=&RandomRangeFloat(Minimum=0, Maximum=1)) {isSetParameters=true}
   &InitializeParticle(
    Lifetime=&Lerp_Float(A=0.7, B=1, Alpha=$Particles.CrystalSizeAlpha)
    Mesh Scale Mode=::NewEnumerator3
    Mesh Uniform Scale=&RandomRangeFloat(Minimum=0.4, Maximum=1.5)
    Position Offset=<0 200 -40>
    UsePositionOffset=true
    Write Material Random=true
   )
   &ShapeLocation(
    Additional Quaternion Rotation=true
    Additional Yaw / Pitch / Roll=<0 0 0.25>
    Cylinder Distribution=::NewEnumerator2
    Cylinder Height=5
    Cylinder Radius=110
    MaskSpawnGroup=false
    Rotation Mode=::NewEnumerator3
    Rotation Quaternion=&OrientMeshToVector(
     Look At Direction=&ReturnCameraVector(Camera Vectors=::NewEnumerator4, Coordinate Space=::World, Origin Position=$Particles.Position)
     Reference Axis Vector=<1 0 0>
    )
    Shape Primitive=::NewEnumerator1
    Spawn Group Mask=1
    U Position=&RandomRangeFloat(Minimum=0, Maximum=1)
    UseOwnerScale=true
    V Position=&RandomRangeFloat(Minimum=0, Maximum=1)
   )
   &CameraOffset(Camera Offset Amount=&RandomRangeFloat(Minimum=0, Maximum=20, Evaluation Type=::NewEnumerator1))
   &AddVelocity(
    Cone Angle=15
    Cone Axis=<0 -1 0>
    Module.UseConeFalloff=false
    Module.UseVelDistribution=false
    Override Seed=false
    Velocity Mode=::NewEnumerator2
    Velocity Speed=&RandomRangeFloat(Minimum=550, Maximum=1500)
   )
  ]
  particleUpdateScript [
   &ParticleState()
   &Drag(Drag=&Lerp_Float(A=4, B=6, Alpha=$Particles.CrystalSizeAlpha), Use Linear Drag=true, Use Rotational Drag=false)
   &VortexForce(
    Origin Pull Amount=1500
    UseOriginOffset=true
    Vortex Axis=<0 1 0>
    Vortex Force Amount=&RandomRangeFloat(Minimum=-1000, Maximum=-2500)
    Vortex Origin Offset:Vector3f=<0 150 -40>
   )
   &SolveForcesAndVelocity(Clamp Velocity=false, Limit Acceleration=false)
   &ScaleMeshSize(Scale Factor=&VectorFromFloat(Value=&FloatFromCurve(FloatCurve=DataInterface() {propertyValues {Curve {keys [<at 0.2 1 RCIM_Cubic>, {arriveTangent=-3.8833, interpMode=RCIM_Cubic, tangentMode=RCTM_Break, time=1, value=0}]}}})))
   &Color(Color=/Niagara/DynamicInputs/Lerp/SmoothLerpOverTime_Color(
     Smooth Value:LinearColor=&ColorFromCurve(ColorCurve=DataInterface() {propertyValues {
        AlphaCurve {keys [<at 0 1 RCIM_Cubic>]}
        BlueCurve {keys [<at 0 0 RCIM_Cubic>,<at 0.1279 0>,<at 0.6396 0 RCIM_Cubic>]}
        GreenCurve {keys [<at 0 10 RCIM_Cubic>,<at 0.1279 0.6562>,<at 0.6396 0.099 RCIM_Cubic>]}
        RedCurve {keys [<at 0 3.4351 RCIM_Cubic>,<at 0.1279 0.1437>,<at 0.6396 0.0313 RCIM_Cubic>]}
       }})
     Target Value:LinearColor=&ColorFromCurve(ColorCurve=DataInterface() {propertyValues {
        AlphaCurve {keys [<at 0 1 RCIM_Cubic>]}
        BlueCurve {keys [<at 0 0 RCIM_Cubic>,<at 0.1279 0>,<at 0.6396 0 RCIM_Cubic>]}
        GreenCurve {keys [<at 0 10 RCIM_Cubic>,<at 0.1279 0.6562>,<at 0.6396 0.099 RCIM_Cubic>]}
        RedCurve {keys [<at 0 4.809 RCIM_Cubic>,<at 0.1279 0.2631>,<at 0.6396 0.0217 RCIM_Cubic>]}
       }})
     Convergence Rate:NiagaraFloat=$User.VariantPicker
    ))
  ]
  rendererClasses [&NiagaraMeshRendererProperties]
  renderers [{rendererClass=&NiagaraMeshRendererProperties, rendererIndex=0}]
 }
 NE_FlashFlare {
  emitterUpdateScript [
   &EmitterState()
   &SpawnBurst_Instantaneous(Spawn Count=2, Use Spawn Probability=false, UseLoopCountLimit=false)
   &SpawnBurst_Instantaneous(Spawn Group=1, Spawn Time=0.1, Use Spawn Probability=false, UseLoopCountLimit=false) {module=SpawnBurst_Instantaneous001}
   &SpawnBurst_Instantaneous(Spawn Count=2, Spawn Group=2, Use Spawn Probability=false, UseLoopCountLimit=false) {module=SpawnBurst_Instantaneous002}
   &SpawnBurst_Instantaneous(Spawn Group=3, Spawn Time=0.05, Use Spawn Probability=false, UseLoopCountLimit=false) {module=SpawnBurst_Instantaneous003}
  ]
  particleSpawnScript [
   /SetVariables(Particles.HandFloat:NiagaraFloat=&SelectFloatFromArray(Direct Array Index=&SpawnGroup(), Float Selection Array=DataInterface() {propertyValues {FloatData [0,0,1,1]}})) {isSetParameters=true}
   &InitializeParticle(
    Color=&LerpLinearColors(StartColor=<rgba 1.8895 3 0.3009 1>, EndColor=<rgba 0.687 3 0 1>, LerpFactor=$User.VariantPicker)
    Lifetime=0.225
    Mass Mode=::NewEnumerator1
    Mesh Scale Mode=::NewEnumerator3
    Mesh Uniform Scale=0.7
    Position=&SimulationPosition()
    Position Mode=::NewEnumerator1
    Position Offset=&Lerp_Vector(
     A=&FN_FrontEndVector3Modifier(Gameplay=<0 200 -40>, Front End=<0 140 20>)
     B=$System.HandPos
     Alpha=$Particles.HandFloat
    )
    Sprite Rotation Angle=&MaskFloatBySpawnGroup(
     Masked Float=&Lerp_Float(A=0.125, B=0.375, Alpha=&ReturnNormalizedExecIndex())
     Passthrough Float=&RandomRangeFloat(Minimum=0, Maximum=1)
    )
    Sprite Rotation Mode=::NewEnumerator1
    Sprite Size=&Multiply_Vector2D(
     A=&Lerp_Vector2(A=<1 1>, B=<0.25 0.25>, Alpha=$Particles.HandFloat)
     B=&MaskVector2dBySpawnGroup(
      Masked Vector2D=&Lerp_Vector2(A=<550 40>, B=<250 40>, Alpha=&ReturnNormalizedExecIndex())
      Passthrough Vector2D=&Vector2DFromFloat(Value=750)
      Spawn Group Mask=0
     )
    )
    Sprite Size Mode=::NewEnumerator4
    UsePositionOffset=true
    Write Material Random=true
   ) {module=InitializeParticle001}
   &CameraOffset(Camera Offset Amount=&Lerp_Float(A=150, B=-25, Alpha=$Particles.HandFloat))
  ]
  particleUpdateScript [
   &ParticleState()
   /SetVariables(Particles.Position=&Lerp_Vector(A=Linked() {linkedVariable {name=Particles.Initial.Position, type=NiagaraPosition}}, B=$System.HandPos, Alpha=$Particles.HandFloat)) {isSetParameters=true}
   &SolveForcesAndVelocity(Clamp Velocity=false, Limit Acceleration=false)
   &ScaleColor(
    Scale Alpha=&FloatFromCurve(FloatCurve=DataInterface() {propertyValues {Curve {keys [
         {arriveTangent=6.4019, interpMode=RCIM_Cubic, leaveTangent=6.4019, tangentMode=RCTM_User, time=0, value=0}
         {arriveTangent=0.002, interpMode=RCIM_Cubic, leaveTangent=0.002, time=0.2, value=0.7}
         {arriveTangent=-0.9006, interpMode=RCIM_Cubic, leaveTangent=-0.9006, tangentMode=RCTM_User, time=1.0011, value=0.002}
        ]}}})
    ScaleA=true
    ScaleRGB=false
   )
   &ScaleSpriteSize(Scale Factor=&Multiply_Vector2D_ByFloat(
     Float=&FloatFromCurve(FloatCurve=DataInterface() {propertyValues {Curve {keys [
          {arriveTangent=0.7585, interpMode=RCIM_Cubic, leaveTangent=0.7585, tangentMode=RCTM_User, time=0, value=0}
          {interpMode=RCIM_Cubic, tangentMode=RCTM_User, time=1, value=1}
         ]}}})
     Vector2D=<1 1>
    ))
   &SpriteRotationRate(Rotation Rate=&UniformRangedFloat(Minimum=-50, Maximum=50))
   /SetVariables(Particles.RayColor:LinearColor=&MultiplyLinearColor(
     Linear Color=&MultiplyLinearColor(
      Linear Color=$Particles.Color
      Scale Factor=&MakeVector4_FromLinearColor(LinearColor=<rgba 1.25 1.25 1.25 1>)
     )
     Scale Factor=&MakeFloat4FromVectorAndFloat(
      "Vector (RGB)"=<1 1 1>
      "Float (Alpha)"=&FloatFromCurve(FloatCurve=DataInterface() {propertyValues {Curve {keys [<at 0.5 1>,<at 1 0>]}}})
     )
    )) {isSetParameters=true}
   /SetVariables(Particles.DarkGlowSize:Vector2f=&Multiply_Vector2D_ByFloat(Float=1, Vector2D=$Particles.SpriteSize)) {isSetParameters=true, module=SetVariables_#9}
   /SetVariables(Particles.DargGlowColor:LinearColor=&ColorFromCurve(ColorCurve=DataInterface() {propertyValues {
       AlphaCurve {keys [<at 0.4134 1 RCIM_Cubic>,<at 1 0 RCIM_Cubic>]}
       BlueCurve {keys [<at 0 0.0343 RCIM_Cubic>]}
       GreenCurve {keys [<at 0 0.1 RCIM_Cubic>]}
       RedCurve {keys [<at 0 0.0372 RCIM_Cubic>]}
      }})) {isSetParameters=true, module=SetVariables_#10}
   &FN_Light_v2(
    Light Exponent=4
    Light Radius:NiagaraFloat=&Lerp_Float(A=110, B=40, Alpha=$Particles.HandFloat)
    LightEnabled=true
    Scale Color:NiagaraFloat=&Multiply_Float(
     A=3
     B=&FloatFromCurve(FloatCurve=DataInterface() {propertyValues {Curve {keys [
          {interpMode=RCIM_Cubic, tangentMode=RCTM_User, time=0, value=1}
          {arriveTangent=-2.5857, interpMode=RCIM_Cubic, leaveTangent=-2.5857, tangentMode=RCTM_User, time=1, value=0}
         ]}}})
    )
   )
  ]
  rendererClasses [&NiagaraSpriteRendererProperties]
  renderers [
   {rendererClass=&NiagaraSpriteRendererProperties}
   {rendererClass=&NiagaraSpriteRendererProperties, rendererIndex=1}
   {rendererClass=&NiagaraSpriteRendererProperties, rendererIndex=2}
  ]
 }
 NE_GlowDecals {
  emitterUpdateScript [
   &EmitterState()
   &SpawnRate(
    SpawnRate=&MakeCustomFloatFromBool(
     Bool=&Boolean_AndOperation(
      Bool1=&SetBoolByFloatComparison(A=$Engine.System.Age, B=$System.SpawnDelay)
      Bool2=&SetBoolByFloatComparison(
       A=$Engine.System.Age
       B=&Add_Float(A=$System.FadeInDuration, B=$System.SpawnDelay)
       Comparison Type=::NewEnumerator6
      )
     )
     True Float=15
     False Float=0
    )
    Use Spawn Probability=false
   )
  ]
  particleSpawnScript [&InitializeParticle(
    Lifetime=&RandomRangeFloat(Minimum=0.5, Maximum=0.6)
    Mesh Scale Mode=::NewEnumerator3
    Position Offset=$System.SpawnOffset
    Sprite Size Mode=::NewEnumerator3
    Uniform Sprite Size=&MakeFloatFromVector2d(Vector 2D=&Multiply_Vector2D_ByFloat(Float=&Lerp_Float(A=250, B=250, Alpha=$System.ProgressAlpha), Vector2D=<1 1>))
    UsePositionOffset=true
    Write Material Random=true
   )]
  particleUpdateScript [
   &ParticleState()
   &Color(Color=&ColorFromCurve(ColorCurve=DataInterface() {propertyValues {
       AlphaCurve {keys [<at 0 0>,<at 0.1971 1 RCIM_Cubic>,<at 1 0 RCIM_Cubic>]}
       BlueCurve {keys [<at 0.2604 0.3242 RCIM_Cubic>,<at 0.5151 0.0326 RCIM_Cubic>,<at 1 0.0051>]}
       GreenCurve {keys [<at 0.2604 2 RCIM_Cubic>,<at 0.5151 1 RCIM_Cubic>,<at 1 0.1562>]}
       RedCurve {keys [<at 0.2604 0 RCIM_Cubic>,<at 0.5151 0 RCIM_Cubic>,<at 1 0>]}
      }}))
   &SolveForcesAndVelocity(Clamp Velocity=false, Limit Acceleration=false)
   &DynamicMaterialParameters(
    Index 0 Param 1=3
    Module.Index0Param0Write=true
    Module.Index0Param1Write=true
    Module.Index0Param2Write=true
    Module.Index0Param3Write=true
   )
   &CameraOffset(Camera Offset Amount=50)
  ]
  rendererClasses [&NiagaraSpriteRendererProperties]
  renderers [{rendererClass=&NiagaraSpriteRendererProperties}]
 }
 NE_GooSprites {
  emitterUpdateScript [
   &EmitterState()
   &SpawnRate(
    SpawnRate=&MakeCustomFloatFromBool(
     Bool=&Boolean_AndOperation(
      Bool1=&SetBoolByFloatComparison(A=$Engine.System.Age, B=$System.SpawnDelay)
      Bool2=&SetBoolByFloatComparison(
       A=$Engine.System.Age
       B=&Add_Float(A=$System.FadeInDuration, B=$System.SpawnDelay)
       Comparison Type=::NewEnumerator6
      )
     )
     True Float=100
     False Float=0
    )
    Use Spawn Probability=false
   )
   &SpawnRate(
    SpawnGroup=1
    SpawnRate=&MakeCustomFloatFromBool(
     Bool=&Boolean_AndOperation(
      Bool1=&SetBoolByFloatComparison(A=$Engine.System.Age, B=&Add_Float(A=$System.SpawnDelay, B=0.15))
      Bool2=&SetBoolByFloatComparison(A=$Engine.System.Age, B=&Add_Float(A=$System.SpawnDelay, B=1.15), Comparison Type=::NewEnumerator6)
     )
     True Float=20
     False Float=0
    )
    Use Spawn Probability=false
   ) {module=SpawnRate001}
  ]
  particleSpawnScript [
   /SetVariables(Particles.HandFloat:NiagaraFloat=&SelectFloatFromArray(Direct Array Index=&SpawnGroup(), Float Selection Array=DataInterface() {propertyValues {FloatData [0,0,1]}})) {isSetParameters=true}
   &InitializeParticle(
    Color=&ColorFromCurve(
     ColorCurve=DataInterface() {propertyValues {
       AlphaCurve {keys [<at 0 1 RCIM_Cubic>]}
       BlueCurve {keys [<at 0.4274 0 RCIM_Cubic>,<at 0.4699 0 RCIM_Cubic>]}
       GreenCurve {keys [<at 0.4274 0.01 RCIM_Cubic>,<at 0.4699 0 RCIM_Cubic>]}
       RedCurve {keys [<at 0.4274 0.0022 RCIM_Cubic>,<at 0.4699 0 RCIM_Cubic>]}
      }}
     CurveIndex=&RandomRangeFloat(Minimum=0, Maximum=1)
    )
    Lifetime=&Lerp_Float(
     A=&MaskFloatBySpawnGroup(
      Masked Float=&RandomRangeFloat(Minimum=0.4, Maximum=0.6)
      Passthrough Float=&RandomRangeFloat(Minimum=1.2, Maximum=1.5)
     )
     B=&RandomRangeFloat(Minimum=1.5, Maximum=2)
     Alpha=$Particles.HandFloat
    )
    Sprite Rotation Mode=::NewEnumerator2
    Sprite Size Mode=::NewEnumerator3
    Uniform Sprite Size=&Lerp_Float(
     A=&MaskFloatBySpawnGroup(
      Masked Float=&RandomRangeFloat(Minimum=8, Maximum=40)
      Passthrough Float=&RandomRangeFloat(Minimum=8, Maximum=15)
     )
     B=&RandomRangeFloat(Minimum=15, Maximum=25)
     Alpha=$Particles.HandFloat
    )
    UsePositionOffset=false
    Write Material Random=true
   )
   &ShapeLocation(
    Cylinder Distribution=::NewEnumerator2
    Cylinder Height=5
    Cylinder Radius=&MaskFloatBySpawnGroup(Masked Float=&Lerp_Float(A=30, B=50, Alpha=$System.ProgressAlpha), Passthrough Float=10)
    MaskSpawnGroup=false
    Offset=&MaskVectorBySpawnGroup(
     Masked Vector=$System.SpawnOffset
     Passthrough Vector=&RandomRangeVector(Minimum=<0 0 -40>, Maximum=<1 150 -50>)
    )
    Offset Mode=::NewEnumerator0
    Rotation Mode=::NewEnumerator2
    Shape Primitive=::NewEnumerator1
    U Position=&RandomRangeFloat(Minimum=0, Maximum=1)
    UseOwnerScale=true
    V Position=&MaskFloatBySpawnGroup(
     Masked Float=&RandomRangeFloat(Minimum=0.5, Maximum=1)
     Passthrough Float=&RandomRangeFloat(Minimum=0.8, Maximum=1)
    )
    Yaw / Pitch / Roll=<0 0 0.25>
   )
   &AddVelocity(
    Velocity=$Particles.ShapeLocation.ShapeNormal
    Velocity Speed Scale=&Multiply_Float(
     A=&RandomRangeFloat(Minimum=350, Maximum=550)
     B=&Lerp_Float(A=0.3, B=0.8, Alpha=$System.ProgressAlpha)
    )
   )
   &AddVelocity(
    Velocity=$Particles.ShapeLocation.ShapeTangent
    Velocity Speed Scale=&Multiply_Float(
     A=&RandomRangeFloat(Minimum=50, Maximum=150)
     B=&Lerp_Float(A=0.8, B=0.6, Alpha=$System.ProgressAlpha)
    )
   ) {module=AddVelocity001}
   &AddVelocity(
    Velocity=<0 1 0>
    Velocity Speed Scale=&MaskFloatBySpawnGroup(Masked Float=0, Passthrough Float=&RandomRangeFloat(Minimum=-350, Maximum=-550))
   ) {module=AddVelocity002}
  ]
  particleUpdateScript [
   &ParticleState()
   &ScaleSpriteSize(
    Scale Sprite Size Mode=::NewEnumerator0
    Uniform Scale Factor=&MaskFloatBySpawnGroup(
     Masked Float=&FloatFromCurve(FloatCurve=DataInterface() {propertyValues {Curve {keys [
          {arriveTangent=8.2127, interpMode=RCIM_Cubic, leaveTangent=8.2127, tangentMode=RCTM_User, time=0, value=0}
          {
           arriveTangent=-0.1296
           arriveTangentWeight=0.232
           interpMode=RCIM_Cubic
           leaveTangent=-0.1296
           leaveTangentWeight=2.2538
           tangentMode=RCTM_User
           time=0.1
           value=1
          }
          {arriveTangent=-0.7869, interpMode=RCIM_Cubic, leaveTangent=-0.7869, tangentMode=RCTM_User, time=1, value=0}
         ]}}})
     Passthrough Float=&FloatFromCurve(FloatCurve=DataInterface() {propertyValues {Curve {keys [
          {arriveTangent=0.9237, interpMode=RCIM_Cubic, leaveTangent=0.9237, tangentMode=RCTM_User, time=0, value=0}
          {
           arriveTangent=-0.1296
           arriveTangentWeight=0.232
           interpMode=RCIM_Cubic
           leaveTangent=-0.1296
           leaveTangentWeight=2.2538
           tangentMode=RCTM_User
           time=0.5
           value=1
          }
          {arriveTangent=-0.7869, interpMode=RCIM_Cubic, leaveTangent=-0.7869, tangentMode=RCTM_User, time=1, value=0}
         ]}}})
    )
   )
   &ScaleColor(ScaleA=true, ScaleRGB=true)
   &DynamicMaterialParameters(
    Index 0 Param 3=1
    Index 0 Param 4=1
    Module.Param0WriteEnabled=true
    Module.Param1WriteEnabled=true
    Module.Param2WriteEnabled=true
    Module.Param3WriteEnabled=true
   )
   &Drag(Drag=3, Use Linear Drag=true, Use Rotational Drag=true)
   &InheritSourceMovement(
    Applied Position Scale=&VectorFromCurve(
     VectorCurve=DataInterface() {propertyValues {XCurve {keys [<at 0 0>,<at 1 1>]}, YCurve {keys [<at 0 0>,<at 1 1>]}, ZCurve {keys [<at 0 0>,<at 1 1>]}}}
     Scale Curve=<-0.2 -0.2 -0.1>
    )
    ApplyForce=false
    ApplyPosition=true
    ApplyPositionFalloff=false
    Source Velocity:Vector3f=&TransformVector(Vector=$Engine.Owner.Velocity, Source Space=::World, Destination Space=::Local)
   )
   &CurlNoiseForce(Noise Strength=&MaskFloatBySpawnGroup(Masked Float=1500, Passthrough Float=350), Pan Noise=false)
   &SolveForcesAndVelocity(Clamp Velocity=false, Limit Acceleration=false)
   &SubUVAnimation(
    Direct Frame Index:NiagaraFloat=&MaskFloatBySpawnGroup(
     Masked Float=&RandomRangeFloat(Minimum=0, Maximum=4)
     Passthrough Float=&RandomRangeFloat(Minimum=0, Maximum=2)
    )
    "Sprite Renderer "=DataInterface() {propertyValues {SpriteRenderer=&self:NE_GooSprites_2.NiagaraSpriteRendererProperties_0}}
    SubUV Animation Mode:ENiagara_SubUVLookupModeV2=::NewEnumerator5
   )
   &CameraOffset(Camera Offset Amount=75)
  ]
  rendererClasses [&NiagaraSpriteRendererProperties]
  renderers [{rendererClass=&NiagaraSpriteRendererProperties}]
 }
 NE_HandRibbon {
  emitterSpawnScript [
   /SetVariables(Emitter.ParticlePerRibbon=20) {isSetParameters=true}
   /SetVariables(Emitter.RibbonNumber=2) {isSetParameters=true}
  ]
  emitterUpdateScript [
   &EmitterState()
   &SpawnBurst_Instantaneous(
    Spawn Count=&Multiply_Int(A=$Emitter.RibbonNumber, B=$Emitter.ParticlePerRibbon)
    Use Spawn Probability=false
    UseLoopCountLimit=false
   )
  ]
  particleSpawnScript [
   &PartitionParticles(
    ID to Partition:NiagaraInt32=&ReturnExecIndex()
    Normalized Index In Partition Includes One=true
    Number Of Partitions=$Emitter.RibbonNumber
    Particles Per Partition:NiagaraInt32=$Emitter.ParticlePerRibbon
    Partition Mode:ENiagara_PartitionMode=::NewEnumerator1
    Write Partition Index to Ribbon Link Order:ENiagara_AttributeSamplingApplyOutput=::NewEnumerator0
    Write Partition to Ribbon ID=::NewEnumerator0
   )
   &InitializeParticle(Color Mode=::NewEnumerator0, Ribbon Width=60, Ribbon Width Mode=::NewEnumerator1, UsePositionOffset=false, Write Material Random=true)
   /SetVariables(Particles.Age:NiagaraFloat=&Multiply_FloatByInt(Float=-0.05, Integer=$Particles.PartitionParticles.Partition)) {isSetParameters=true}
  ]
  particleUpdateScript [
   &ParticleState()
   &RotateAroundPoint(
    Radius:NiagaraFloat=&FN_FrontEndFloatModifier(Gameplay=100, Front End=75)
    Rotation Center:NiagaraPosition=&FN_FrontEndVector3Modifier(
     Gameplay=&Add_Vector(
      A=<0 0 -40>
      B=&Lerp_Vector(A=<0 -50 -40>, B=<0 300 -50>, Alpha=$Particles.PartitionParticles.Normalized Index In Partition)
     )
     Front End=&Add_Vector(
      A=<0 0 0>
      B=&Lerp_Vector(A=<0 -250 -30>, B=<0 150 -20>, Alpha=$Particles.PartitionParticles.Normalized Index In Partition)
     )
    )
    Rotation Phase:NiagaraFloat=&Add_Float(
     A=&Add_Float(
      A=&Multiply_Float(
       A=$Particles.PartitionParticles.Normalized Index In Partition
       B=&Multiply_Float(
        A=3
        B=&SelectFloatFromArray(Direct Array Index=$Particles.PartitionParticles.Partition, Float Selection Array=DataInterface() {propertyValues {FloatData [1,1]}})
       )
      )
      B=&Multiply_FloatByInt(
       Float=&Divide_Float(A=1, B=&MakeFloatFromInt(Int=$Emitter.RibbonNumber))
       Integer=$Particles.PartitionParticles.Partition
      )
     )
     B=&Multiply_Float(A=$Particles.Age, B=-0.1)
    )
    Set Particle Position=false
    X Axis:Vector3f=<0 0 1>
    Y Axis:Vector3f=<1 0 0>
   )
   /SetVariables(Particles.Position=&Lerp_Vector(
     A=$System.HandPos
     B=$Output.RotateAroundPoint.OutputPosition
     Alpha=&FloatFromCurve(
      FloatCurve=DataInterface() {propertyValues {Curve {keys [<at 0 0>,<at 0.5 1>]}}}
      CurveIndex=$Particles.PartitionParticles.Normalized Index In Partition
     )
    )) {isSetParameters=true}
   &ScaleRibbonWidth(Ribbon Width Scale=&FloatFromCurve(
     FloatCurve=DataInterface() {propertyValues {Curve {keys [<at 0 0>,<at 0.4 1>,<at 0.8 1>,<at 1 0>]}}}
     CurveIndex=$Particles.PartitionParticles.Normalized Index In Partition
    ))
   &Color(
    Color=&ColorFromCurve(ColorCurve=DataInterface() {propertyValues {
       AlphaCurve {keys [<at 0 1 RCIM_Cubic>]}
       BlueCurve {keys [<at 0 1 RCIM_Cubic>,<at 1 1 RCIM_Cubic>]}
       GreenCurve {keys [<at 0 1 RCIM_Cubic>,<at 1 1 RCIM_Cubic>]}
       RedCurve {keys [<at 0 1 RCIM_Cubic>,<at 1 1 RCIM_Cubic>]}
      }})
    Scale Alpha=&Multiply_Float(
     A=&FloatFromCurve(
      FloatCurve=DataInterface() {propertyValues {Curve {keys [
          {interpMode=RCIM_Cubic, tangentMode=RCTM_User, time=-0.3, value=0}
          {interpMode=RCIM_Cubic, tangentMode=RCTM_User, time=-0.2, value=1}
          {interpMode=RCIM_Cubic, tangentMode=RCTM_User, time=-0.1, value=0}
         ]}}}
      CurveIndex=&Add_Float(A=$Particles.PartitionParticles.Normalized Index In Partition, B=&Multiply_Float(A=$Particles.Age, B=-1))
     )
     B=1
    )
   ) {enabled=false}
   &DynamicMaterialParameters(
    Index 1 Param 2=&FloatFromCurve(
     FloatCurve=DataInterface() {propertyValues {Curve {keys [<at 0 0>,<at 1 1>]}}}
     CurveIndex=&Multiply_Float(A=$Particles.Age, B=1)
    )
    Module.Index1Param0Write=true
    Module.Index1Param1Write=true
    Module.Index1Param2Write=true
    Module.Index1Param3Write=true
    Write Parameter Index 0=false
    Write Parameter Index 1=true
   )
  ]
  rendererClasses [&NiagaraSpriteRendererProperties, &NiagaraRibbonRendererProperties]
  renderers [
   {rendererClass=&NiagaraSpriteRendererProperties}
   {rendererClass=&NiagaraRibbonRendererProperties, rendererIndex=1}
  ]
 }
 NE_LongGlow {
  emitterUpdateScript [
   &EmitterState()
   &SpawnBurst_Instantaneous(Spawn Time=0.2, Use Spawn Probability=false, UseLoopCountLimit=false)
  ]
  particleSpawnScript [
   &InitializeParticle(
    Color=<rgba 0.1564 1 0 1>
    Lifetime=0.35
    Mass Mode=::NewEnumerator1
    Mesh Scale Mode=::NewEnumerator3
    Mesh Uniform Scale=0.7
    Position=&SimulationPosition()
    Position Mode=::NewEnumerator1
    Position Offset=<0 100 -40>
    Sprite Size Mode=::NewEnumerator3
    Uniform Sprite Size=550
    UsePositionOffset=true
    Write Material Random=true
   ) {module=InitializeParticle001}
   &CameraOffset(Camera Offset Amount=50)
   /SetVariables(Particles.SpriteAlignment=<0 1 0>) {isSetParameters=true}
  ]
  particleUpdateScript [
   &ParticleState()
   &SolveForcesAndVelocity(Clamp Velocity=false, Limit Acceleration=false)
   &ScaleColor(
    Scale Alpha=&FloatFromCurve(FloatCurve=DataInterface() {propertyValues {Curve {keys [
         {arriveTangent=6.4019, interpMode=RCIM_Cubic, leaveTangent=6.4019, tangentMode=RCTM_User, time=0, value=0}
         {arriveTangent=0.002, interpMode=RCIM_Cubic, leaveTangent=0.002, time=0.2, value=0.7}
         {arriveTangent=-0.9006, interpMode=RCIM_Cubic, leaveTangent=-0.9006, tangentMode=RCTM_User, time=1.0011, value=0.002}
        ]}}})
    ScaleA=true
    ScaleRGB=false
   )
   &SpriteRotationRate(Rotation Rate=&UniformRangedFloat(Minimum=-50, Maximum=50))
   /SetVariables(Particles.DarkGlowSize:Vector2f=&Multiply_Vector2D_ByFloat(Float=3, Vector2D=$Particles.SpriteSize)) {isSetParameters=true, module=SetVariables_#9}
   /SetVariables(Particles.DargGlowColor:LinearColor=&ColorFromCurve(ColorCurve=DataInterface() {propertyValues {
       AlphaCurve {keys [<at 0.4134 1 RCIM_Cubic>,<at 1 0 RCIM_Cubic>]}
       BlueCurve {keys [<at 0 0 RCIM_Cubic>]}
       GreenCurve {keys [<at 0 0 RCIM_Cubic>]}
       RedCurve {keys [<at 0 0 RCIM_Cubic>]}
      }})) {isSetParameters=true, module=SetVariables_#10}
   &FN_Light_v2(
    Light Exponent=1
    Light Radius=350
    LightEnabled=true
    Position Offset:Vector3f=<0 0 -40>
    Scale Color:NiagaraFloat=&Multiply_Float(
     A=15
     B=&FloatFromCurve(FloatCurve=DataInterface() {propertyValues {Curve {keys [
          {interpMode=RCIM_Cubic, tangentMode=RCTM_User, time=0, value=1}
          {arriveTangent=-2.5857, interpMode=RCIM_Cubic, leaveTangent=-2.5857, tangentMode=RCTM_User, time=1, value=0}
         ]}}})
    )
   )
   &DynamicMaterialParameters(
    Index 0 Param 2=&MakeFloatFromVector(Vector=&TransformVector(Vector=<0 1 0>, Destination Space=::World), Channel=::NewEnumerator1)
    Index 0 Param 4=500
    Module.Index0Param0Write=true
    Module.Index0Param1Write=true
    Module.Index0Param2Write=true
    Module.Index0Param3Write=true
   )
   /SetVariables(Particles.DarkGlowDynMat:Vector4f=&MakeVector4(
     X=0
     Y=&MakeFloatFromVector(Vector=&TransformVector(Vector=<0 1 0>, Destination Space=::World), Channel=::NewEnumerator1)
     Z=0
     W=1200
    )) {isSetParameters=true}
  ]
  rendererClasses [&NiagaraSpriteRendererProperties, &NiagaraLightRendererProperties]
  renderers [
   {rendererClass=&NiagaraSpriteRendererProperties}
   {rendererClass=&NiagaraSpriteRendererProperties, rendererIndex=1}
   {rendererClass=&NiagaraLightRendererProperties, rendererIndex=2}
  ]
 }
 NE_SpiralTrails001 {
  emitterUpdateScript [
   &EmitterState()
   &SpawnRate(
    SpawnRate=&MakeCustomFloatFromBool(
     Bool=&Boolean_AndOperation(
      Bool1=&SetBoolByFloatComparison(A=$Engine.System.Age, B=$System.SpawnDelay)
      Bool2=&SetBoolByFloatComparison(
       A=$Engine.System.Age
       B=&Add_Float(A=$System.FadeInDuration, B=$System.SpawnDelay)
       Comparison Type=::NewEnumerator6
      )
     )
     True Float=35
     False Float=0
    )
    Use Spawn Probability=false
   )
   &SpawnRate(
    SpawnGroup=1
    SpawnRate=&MakeCustomFloatFromBool(
     Bool=&Boolean_AndOperation(
      Bool1=&SetBoolByFloatComparison(A=$Engine.System.Age, B=&Add_Float(A=$System.SpawnDelay, B=0.15))
      Bool2=&SetBoolByFloatComparison(A=$Engine.System.Age, B=&Add_Float(A=$System.SpawnDelay, B=1.15), Comparison Type=::NewEnumerator6)
     )
     True Float=10
     False Float=0
    )
    Use Spawn Probability=false
   ) {module=SpawnRate001}
  ]
  particleSpawnScript [
   &InitializeParticle(
    Lifetime=&MaskFloatBySpawnGroup(
     Masked Float=&RandomRangeFloat(Minimum=0.8, Maximum=1)
     Passthrough Float=&RandomRangeFloat(Minimum=1, Maximum=1.5)
    )
    Mesh Scale Mode=::NewEnumerator3
    Mesh Uniform Scale=&Multiply_Float(
     A=&MaskFloatBySpawnGroup(
      Masked Float=&Lerp_Float(
       A=2
       B=2.5
       Alpha=&FloatFromCurve(
        FloatCurve=DataInterface() {propertyValues {Curve {keys [
            {interpMode=RCIM_Cubic, tangentMode=RCTM_User, time=0.5, value=0}
            {arriveTangent=7.2351, interpMode=RCIM_Cubic, leaveTangent=7.2351, tangentMode=RCTM_User, time=1, value=1}
           ]}}}
        CurveIndex=$System.ProgressAlpha
       )
      )
      Passthrough Float=&RandomRangeFloat(Minimum=1.1, Maximum=1.3)
     )
     B=&FN_FrontEndFloatModifier(Gameplay=1, Front End=0.7)
    )
    Position Offset=&MaskVectorBySpawnGroup(
     Masked Vector=$System.SpawnOffset
     Passthrough Vector=&RandomRangeVector(Minimum=<0 100 -40>, Maximum=<1 -75 -50>)
    )
    Sprite Size Mode=::NewEnumerator3
    UsePositionOffset=true
    Write Material Random=true
   )
   &InitialMeshOrientation(
    Mesh Orientation Mode=::NewEnumerator7
    Rotation=&RandomRangeVector(Minimum=<-0.1 0 0>, Maximum=<0.1 1 0>)
    Use Rotation Vector=true
   )
   /SetVariables(Particles.GlowDecalSize:Vector2f=&Multiply_Vector2D_ByFloat(Float=&Lerp_Float(A=250, B=100, Alpha=$System.ProgressAlpha), Vector2D=<1 1>)) {isSetParameters=true}
   &AddVelocity(Velocity=&MaskVectorBySpawnGroup(Masked Vector=<0 0 0>, Passthrough Vector=&RandomRangeVector(Minimum=<0 -50 0>, Maximum=<0 -75 0>)))
   &DynamicMaterialParameters(
    Index 1 Param 1=&Lerp_Float(A=0.5, B=1, Alpha=$System.ProgressAlpha)
    Module.Index1Param0Write=true
    Module.Index1Param1Write=false
    Module.Index1Param2Write=false
    Module.Index1Param3Write=false
    Write Parameter Index 0=false
    Write Parameter Index 1=true
   ) {module=DynamicMaterialParameters001}
  ]
  particleUpdateScript [
   &ParticleState()
   &ScaleMeshSize(Scale Factor=&VectorFromFloat(Value=&MaskFloatBySpawnGroup(
      Masked Float=&FloatFromCurve(FloatCurve=DataInterface() {propertyValues {Curve {keys [
           {arriveTangent=2.4544, interpMode=RCIM_Cubic, leaveTangent=2.4544, tangentMode=RCTM_User, time=0, value=0.2}
           {interpMode=RCIM_Cubic, tangentMode=RCTM_User, time=1, value=1}
          ]}}})
      Passthrough Float=&FloatFromCurve(FloatCurve=DataInterface() {propertyValues {Curve {keys [
           {arriveTangent=0.3327, interpMode=RCIM_Cubic, leaveTangent=0.3327, tangentMode=RCTM_User, time=0, value=1}
           {interpMode=RCIM_Cubic, tangentMode=RCTM_User, time=1, value=1.5}
          ]}}})
     )))
   &Color(Color=&MaskLinearColorBySpawnGroup(
     Masked Linear Color=&ColorFromCurve(ColorCurve=DataInterface() {propertyValues {
        AlphaCurve {keys [<at 0 1 RCIM_Cubic>,<at 1 0 RCIM_Cubic>]}
        BlueCurve {keys [<at 0 0 RCIM_Cubic>,<at 0.2409 0.5>,<at 0.7276 0.1 RCIM_Cubic>]}
        GreenCurve {keys [<at 0 10 RCIM_Cubic>,<at 0.2409 0.5>,<at 0.7276 0.1 RCIM_Cubic>]}
        RedCurve {keys [<at 0 5.0901 RCIM_Cubic>,<at 0.2409 0.5>,<at 0.7276 0.1 RCIM_Cubic>]}
       }})
     Passthrough Linear Color=&ColorFromCurve(ColorCurve=DataInterface() {propertyValues {
        AlphaCurve {keys [<at 0 1 RCIM_Cubic>,<at 1 0 RCIM_Cubic>]}
        BlueCurve {keys [<at 0 1 RCIM_Cubic>,<at 0.2409 0.5>,<at 0.7276 0.1 RCIM_Cubic>]}
        GreenCurve {keys [<at 0 1 RCIM_Cubic>,<at 0.2409 0.5>,<at 0.7276 0.1 RCIM_Cubic>]}
        RedCurve {keys [<at 0 1 RCIM_Cubic>,<at 0.2409 0.5>,<at 0.7276 0.1 RCIM_Cubic>]}
       }})
    ))
   &SolveForcesAndVelocity(Clamp Velocity=false, Limit Acceleration=false)
   &DynamicMaterialParameters(
    Index 0 Param 1=&RandomRangeFloat(Minimum=0, Maximum=1)
    Index 1 Param 3=&MaskFloatBySpawnGroup(
     Masked Float=&FloatFromCurve(FloatCurve=DataInterface() {propertyValues {Curve {keys [
          {arriveTangent=2.6527, interpMode=RCIM_Cubic, leaveTangent=2.6527, tangentMode=RCTM_User, time=0, value=0}
          {interpMode=RCIM_Cubic, tangentMode=RCTM_User, time=1, value=1.5}
         ]}}})
     Passthrough Float=&FloatFromCurve(FloatCurve=DataInterface() {propertyValues {Curve {keys [
          <at 0 1.5 RCIM_Cubic>
          {arriveTangent=0.0527, interpMode=RCIM_Cubic, leaveTangent=0.0527, tangentMode=RCTM_User, time=0.35, value=0.5}
          {interpMode=RCIM_Cubic, tangentMode=RCTM_User, time=1, value=1.5}
         ]}}})
    )
    Module.Index0Param0Write=true
    Module.Index0Param1Write=true
    Module.Index0Param2Write=true
    Module.Index0Param3Write=true
    Module.Index1Param0Write=false
    Module.Index1Param1Write=true
    Module.Index1Param2Write=true
    Module.Index1Param3Write=true
    Write Parameter Index 1=true
   )
   /SetVariables(Particles.GlowDecalSize:Vector2f=&Multiply_Vector2D_ByFloat(Float=1, Vector2D=$Particles.Initial.GlowDecalSize)) {isSetParameters=true, module=SetVariables_#11}
   /SetVariables(Particles.DecalDynMat:Vector4f=<3 1 1 0>) {isSetParameters=true}
   /SetVariables(Particles.GlowDecalColor:LinearColor=&ColorFromCurve(ColorCurve=DataInterface() {propertyValues {
       AlphaCurve {keys [<at 0 0>,<at 0.2685 1 RCIM_Cubic>,<at 1 0 RCIM_Cubic>]}
       BlueCurve {keys [<at 0.2604 0.3242 RCIM_Cubic>,<at 0.5151 0.0326 RCIM_Cubic>,<at 1 0.0051>]}
       GreenCurve {keys [<at 0.2604 2 RCIM_Cubic>,<at 0.5151 1 RCIM_Cubic>,<at 1 0.1562>]}
       RedCurve {keys [<at 0.2604 0 RCIM_Cubic>,<at 0.5151 0 RCIM_Cubic>,<at 1 0>]}
      }})) {isSetParameters=true}
   &CameraOffset(Camera Offset Amount=&MaskFloatBySpawnGroup(Masked Float=75, Passthrough Float=0))
  ]
  rendererClasses [&NiagaraMeshRendererProperties]
  renderers [{rendererClass=&NiagaraMeshRendererProperties, rendererIndex=0}]
 }
}
summary {
 systemName=NS_Glider_AnglePatch_Deploy
 emitters [
  {emitterName=NE_GlowDecals, rendererClasses [&NiagaraSpriteRendererProperties]}
  {emitterName=NE_LongGlow, rendererClasses [&NiagaraSpriteRendererProperties, &NiagaraLightRendererProperties]}
  {emitterName=NE_FlashFlare, rendererClasses [&NiagaraSpriteRendererProperties]}
  {emitterName=NE_CrystalDebris, rendererClasses [&NiagaraMeshRendererProperties]}
  {emitterName=NE_GooSprites, rendererClasses [&NiagaraSpriteRendererProperties]}
  {emitterName=NE_SpiralTrails001, rendererClasses [&NiagaraMeshRendererProperties]}
  {
   emitterName=NE_HandRibbon
   rendererClasses [&NiagaraSpriteRendererProperties, &NiagaraRibbonRendererProperties]
  }
 ]
 userVariables [
  User.bIsFrontEnd {type=NiagaraBool}
  User.bIsFrontEndPreview {type=NiagaraBool}
  User.PawnSkeletal {
   type=NiagaraDataInterfaceSkeletalMesh
   defaultValue {
    struct=NiagaraExt_VariableValue_DataInterface
    value {dataInterface=&self:NiagaraDataInterfaceSkeletalMesh_0}
   }
  }
  User.VariantPicker {type=NiagaraFloat, defaultValue=0, description=""}
 ]
}
