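# Niagara trail effect for the AnglePatch glider, with straight and spiraling ribbons, flickering jet glows with lights, goo drops and speed lines.
# Persistent ribbon particles, three strands per emitter, anchor to skeletal-mesh bones and lerp toward trailing goals, with sine/cosine offsets forming the spiral.
# Goo drops follow spiral ribbons via particle reads.
# Smoothed forward-dot and fade-in values drive intensity and length, with separate front-end tuning.
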NiagaraSystem /BRCosmetics/Gadgets/Assets/VinderTech_GliderChute/Glider_AnglePatch/FX/NS_Glider_AnglePatch_Trail
dependencies {
 usedDataInterfaces [
  &NiagaraDataInterfaceColorCurve
  &NiagaraDataInterfaceCurve
  &NiagaraDataInterfaceCamera
  &NiagaraDataInterfaceVector2DCurve
  &NiagaraDataInterfaceArrayFloat3
  &NiagaraDataInterfaceParticleRead
  &NiagaraDataInterfaceSpriteRendererInfo
  &NiagaraDataInterfaceSkeletalMesh
 ]
 usedDynamicInputs [&ClampFloat, &SetBoolByFloatComparison]
 usedModules [
  &self:SystemSpawnScript.SystemScriptSource.SystemScriptGraph.NiagaraNodeAssignment_4.SetVariables_#1
  &self:SystemSpawnScript.SystemScriptSource.SystemScriptGraph.NiagaraNodeAssignment_5.SetVariables_#2
  &self:SystemSpawnScript.SystemScriptSource.SystemScriptGraph.NiagaraNodeAssignment_2.SetVariables_#3
  &self:SystemSpawnScript.SystemScriptSource.SystemScriptGraph.NiagaraNodeAssignment_7.SetVariables_#4
  &self:SystemSpawnScript.SystemScriptSource.SystemScriptGraph.NiagaraNodeAssignment_0.SetVariables_#5
  &self:SystemSpawnScript.SystemScriptSource.SystemScriptGraph.NiagaraNodeAssignment_3.SetVariables_#6
  &SystemState
  &self:SystemSpawnScript.SystemScriptSource.SystemScriptGraph.NiagaraNodeAssignment_6.SetVariables_#7
  &Timeline
  &self:SystemSpawnScript.SystemScriptSource.SystemScriptGraph.NiagaraNodeAssignment_1.SetVariables_#8
 ]
}
emitters {
 NE_GooDrops001 {
  emitterUpdateScript [
   &EmitterState()
   &SpawnRate(
    Spawn Probability=0.25
    SpawnRate=&Multiply_Float(
     A=&Multiply_Float(
      A=&Lerp_Float(A=10, B=70, Alpha=$System.SmoothedForwardDot)
      B=&FloatFromCurve(
       FloatCurve=DataInterface() {propertyValues {Curve {keys [<at 0 1 RCIM_Constant>,<at 0.05 0.2 RCIM_Constant>,<at 0.2 1.175 RCIM_Constant>,<at 0.5 0.2 RCIM_Constant>,<at 0.8 1 RCIM_Constant>,<at 0.9 0.2 RCIM_Constant>]}}}
       CurveIndex=&ModuloFloat(A=$System.LoopedAge, B=1)
      )
     )
     B=$System.SystemFadeIn
    )
    Use Spawn Probability=false
   )
  ]
  particleSpawnScript [
   &PartitionParticles(
    ID to Partition:NiagaraInt32=&ReturnExecIndex()
    Number Of Partitions=3
    Partition Mode:ENiagara_PartitionMode=::NewEnumerator5
   )
   &InitializeParticle(
    Color=&ColorFromCurve(
     ColorCurve=DataInterface() {propertyValues {
       AlphaCurve {keys [<at 0 1 RCIM_Cubic>]}
       BlueCurve {keys [<at 0.4274 0 RCIM_Cubic>,<at 0.4699 0 RCIM_Cubic>]}
       GreenCurve {keys [<at 0.4274 0.01 RCIM_Cubic>,<at 0.4699 0 RCIM_Cubic>]}
       RedCurve {keys [<at 0.4274 0.0022 RCIM_Cubic>,<at 0.4699 0 RCIM_Cubic>]}
      }}
     CurveIndex=&RandomRangeFloat(Minimum=0, Maximum=1)
    )
    Lifetime=&Lerp_Float(
     A=&RandomRangeFloat(Minimum=0.9, Maximum=1.1)
     B=&RandomRangeFloat(Minimum=0.2, Maximum=0.3)
     Alpha=$System.SmoothedForwardDot
    )
    Sprite Rotation Mode=::NewEnumerator2
    Sprite Size Mode=::NewEnumerator3
    Uniform Sprite Size=&RandomRangeFloat(Minimum=6, Maximum=16)
    UsePositionOffset=true
    Write Material Random=true
   )
   /SetVariables(Particles.Age:NiagaraFloat=&Multiply_FloatByInt(Float=-0.2563, Integer=1)) {isSetParameters=true}
   &ShapeLocation(Apply To Particle Position=::NewEnumerator1, MaskSpawnGroup=false, Override Seed=false, Sphere Radius=20, UseOwnerScale=true)
   /SetVariables() {isSetParameters=true}
   &CameraOffset(Camera Offset Amount=20)
  ]
  particleUpdateScript [
   &ParticleState()
   &SpriteRotationRate(Rotation Rate=&Multiply_Float(
     A=&RandomRangeFloat(Minimum=250, Maximum=160)
     B=&UniformAOrBFloat(A=-1, Evaluation Type=::NewEnumerator0, Recalculate Random Each Loop=true, Override Seed=false)
    ))
   &SkeletalMeshLocation(
    Bone Sampling Mode=::NewEnumerator4
    MaskSpawnGroup=false
    Position Sampling=::NewEnumerator2
    Sampled Position Offset=<-2 0 8>
    Skeletal Mesh=$System.Skeletal Mesh
    Use Sampled Mesh Attribute as Ribbon ID=false
   )
   /SetVariables(Particles.HandleDropOrientation:Vector3f=&RotateVectorByQuaternion(
     Vector To Rotate=&SelectVectorFromArray(
      Array Sampling Mode=::NewEnumerator0
      Direct Array Index=?
      Vector Selection Array=DataInterface() {propertyValues {FloatData [<0 1 1>, <0 -1 1>, <0 1 -1>, <0 -1 -1>]}}
      Random Seed=0
     )
     Quaternion=$Particles.SkeletalMeshLocation.SampledRotation
    )) {isSetParameters=true}
   &self:FollowRibbon(
    Alpha:NiagaraFloat=&FloatFromCurve(FloatCurve=DataInterface() {propertyValues {Curve {keys [<at 0 0.85>,<at 1 0.2>]}}})
    ParticlePerRibbon=8
    RibbonNumber:NiagaraInt32=$Particles.PartitionParticles.Partition
    RibbonReader=DataInterface() {propertyValues {EmitterBinding {bindingMode=Other, emitterName=NE_SpiralRibbon}}}
   ) {module=FollowRibbon}
   /SetVariables(Particles.Position=&Lerp_Vector(
     A=Linked() {linkedVariable {name=Particles.Position, type=NiagaraPosition}}
     B=&AddVectorToPosition(
      Position=$Output.FollowRibbon.RibbonPos
      Vector=&Multiply_Vector(A=$Output.FollowRibbon.DirectionVector, B=&RandomRangeVector(Minimum=<10 10 0>, Maximum=<-10 -10 0>))
     )
     Alpha=&Multiply_Float(A=$Engine.DeltaTime, B=8)
    )) {isSetParameters=true}
   /SetVariables(Particles.Position=&AddVectorToPosition(
     Position=$Output.FollowRibbon.RibbonPos
     Vector=&Multiply_Vector(A=$Output.FollowRibbon.DirectionVector, B=&RandomRangeVector(Minimum=<10 10 0>, Maximum=<-10 -10 0>))
    )) {enabled=false, isSetParameters=true}
   /SetVariables(Particles.Velocity=&MaskVectorBySpawnGroup(
     Masked Vector=&Lerp_Vector(
      A=$Particles.Velocity
      B=&Multiply_VectorByFloat(Vector=$Output.FollowRibbon.DirectionVector, Float=-1)
      Alpha=&Multiply_Float(A=$Engine.DeltaTime, B=5)
     )
     Passthrough Vector=$Particles.HandleDropOrientation
    )) {isSetParameters=true}
   &ScaleSpriteSize(Uniform Curve Sprite Scale=DataInterface() {propertyValues {Curve {keys [
        {arriveTangent=8.2127, interpMode=RCIM_Cubic, leaveTangent=8.2127, tangentMode=RCTM_User, time=0, value=0}
        {
         arriveTangent=0.4596
         arriveTangentWeight=0.232
         interpMode=RCIM_Cubic
         leaveTangent=0.4596
         leaveTangentWeight=2.2538
         tangentMode=RCTM_User
         time=0.1
         value=1
        }
        {arriveTangent=0.0486, interpMode=RCIM_Cubic, leaveTangent=0.0486, tangentMode=RCTM_User, time=0.75, value=1}
        {arriveTangent=-4.4982, interpMode=RCIM_Cubic, leaveTangent=-4.4982, tangentMode=RCTM_User, time=1, value=0}
       ]}}})
   &ScaleColor(ScaleA=true, ScaleRGB=true)
   &DynamicMaterialParameters(
    Index 0 Param 3=1
    Index 0 Param 4=&MaskFloatBySpawnGroup(Masked Float=0, Passthrough Float=0)
    Module.Param0WriteEnabled=true
    Module.Param1WriteEnabled=true
    Module.Param2WriteEnabled=true
    Module.Param3WriteEnabled=true
   )
   &SubUVAnimation(
    Evaluation Type=::NewEnumerator0
    "Sprite Renderer "=DataInterface() {propertyValues {SpriteRenderer=&self:NE_GooDrops001_2.NiagaraSpriteRendererProperties_1}}
    SubUV Animation Mode:ENiagara_SubUVLookupModeV2=::NewEnumerator2
    UseEndFrame=false
    UseStartFrame=false
   )
  ]
  rendererClasses [&NiagaraSpriteRendererProperties]
  renderers [{rendererClass=&NiagaraSpriteRendererProperties}]
 }
 NE_JetGlow {
  emitterUpdateScript [
   &EmitterState(
    Inactive Response=::NewEnumerator1
    Life Cycle Mode=::NewEnumerator1
    Loop Behavior=::NewEnumerator1
    Loop Duration Mode=::NewEnumerator1
    UseLoopDelay=false
   )
   &SpawnBurst_Instantaneous(Spawn Count=3, Use Spawn Probability=false, UseLoopCountLimit=false)
   &SpawnBurst_Instantaneous(Spawn Group=1, Use Spawn Probability=false, UseLoopCountLimit=false) {module=SpawnBurst_Instantaneous001}
  ]
  particleSpawnScript [
   &PartitionParticles(ID to Partition:NiagaraInt32=&ReturnExecIndex(), Number Of Partitions=3)
   &InitializeParticle(
    Color=&MultiplyLinearColor(Linear Color=$System.StraightColorMid, Scale Factor=<1 1 1 0.5>)
    Lifetime=0.5
    Mesh Scale=&FN_FrontEndVector3Modifier(Gameplay=<0.23 0.23 0.3>, Front End=<0.2 0.2 0.2>)
    Mesh Scale Mode=::NewEnumerator4
    Sprite Size Mode=::NewEnumerator3
    Uniform Sprite Size=65
    UsePositionOffset=true
    Write Material Random=true
   )
   &InitialMeshOrientation(Mesh Orientation Mode=::NewEnumerator7, Rotation=<0.25 0 0>, Use Rotation Vector=true)
   /SetVariables(Particles.VisibilityTag=&SpawnGroup()) {isSetParameters=true}
  ]
  particleUpdateScript [
   &ParticleState(Kill Particles When Lifetime Has Elapsed=false)
   /SetVariables(Particles.LightFlicker:NiagaraFloat=&Add_Float(
     A=1
     B=&Multiply_Float(
      A=&Multiply_Float(
       A=&Sine(Normalized Angle=$System.Age, Period=0.3542, Scale=1, Bias=1)
       B=&Sine(Normalized Angle=$System.Age, Period=0.2354, Scale=1, Bias=1)
      )
      B=0.2
     )
    )) {isSetParameters=true}
   &SkeletalMeshLocation(
    Bone / Socket Index=&MakeCustomIntFromBool(
     Bool=&SetBoolByIntComparison(A=&SpawnGroup(), B=0, Comparison Type=::NewEnumerator4)
     True INT=$Particles.PartitionParticles.Partition
     False INT=3
    )
    Bone Sampling Mode=::NewEnumerator4
    MaskSpawnGroup=false
    Sampled Position Offset=&Add_Vector(A=$System.BaseOffset, B=<0 0 0>)
    Skeletal Mesh=$System.Skeletal Mesh
   )
   /SetVariables(Particles.JetColor:LinearColor=&MultiplyLinearColor(
     Linear Color=<rgba 1 1 1 1>
     Scale Factor=&MakeFloat4FromVectorAndFloat("Vector (RGB)"=<1 1 1>, "Float (Alpha)"=$System.SystemFadeIn)
    )) {isSetParameters=true}
   /SetVariables(Particles.JetSize:Vector2f=&Multiply_Vector2D_ByFloat(Float=45, Vector2D=<1 1>)) {isSetParameters=true}
   /SetVariables(Particles.JetDynMat:Vector4f=&MakeVector4(X=0, Y=$System.SmoothedForwardDot, Z=0, W=0)) {isSetParameters=true}
   /SetVariables(Particles.SpriteFacing=<0 1 0>) {isSetParameters=true}
   /SetVariables(Particles.DecalColor:LinearColor=&MultiplyLinearColor(
     Linear Color=&MultiplyLinearColor(
      Linear Color=<rgba 0 1 0.0923 0.02>
      Scale Factor=&MakeFloat4FromVectorAndFloat(
       "Vector (RGB)"=&VectorFromFloat(Value=&Lerp_Float(A=0.1, B=1, Alpha=$System.SmoothedForwardDot))
       "Float (Alpha)"=1
      )
     )
     Scale Factor=&MakeFloat4FromVectorAndFloat("Vector (RGB)"=<1 1 1>, "Float (Alpha)"=&Multiply_Float(A=$System.SystemFadeIn, B=$Particles.LightFlicker))
    )) {isSetParameters=true}
   /SetVariables(Particles.DecalDynMat:Vector4f=<0.5 0 0 0>) {isSetParameters=true}
   /SetVariables(Particles.DecalCamOffset=20) {isSetParameters=true}
   &DynamicMaterialParameters(
    Index 0 Param 2=1
    Module.Index0Param0Write=true
    Module.Index0Param1Write=true
    Module.Index0Param2Write=true
    Module.Index0Param3Write=true
   )
   &ScaleColor(Scale Alpha=$System.SystemFadeIn, ScaleA=true, ScaleRGB=true)
   &FN_Light_Attributes(
    Color=<rgba 0.4069 1 0.413 1>
    Create Color Attribute=true
    Create Exponent Attribute=true
    Create Position Attribute=true
    Create Radius Attribute=true
    Exponent=2
    Position=&AddVectorToPosition(Position=$Particles.Position, Vector=<0 0 0>)
    Radius=&Multiply_Float(A=&Lerp_Float(A=60, B=70, Alpha=$System.SmoothedForwardDot), B=$System.SystemFadeIn)
    Scale Color=&Multiply_Float(A=&Lerp_Float(A=4, B=20, Alpha=$System.SmoothedForwardDot), B=$Particles.LightFlicker)
   )
  ]
  rendererClasses [
   &NiagaraSpriteRendererProperties
   &NiagaraLightRendererProperties
   &NiagaraMeshRendererProperties
  ]
  renderers [
   {rendererClass=&NiagaraSpriteRendererProperties}
   {rendererClass=&NiagaraLightRendererProperties, rendererIndex=1}
   {rendererClass=&NiagaraMeshRendererProperties, rendererIndex=2}
   {rendererClass=&NiagaraSpriteRendererProperties, rendererIndex=3}
  ]
 }
 NE_Speed_Lines {
  emitterUpdateScript [
   &EmitterState()
   &SpawnRate(SpawnRate=&FN_FrontEndFloatModifier(Gameplay=100, Front End=50), Use Spawn Probability=false) {module=SpawnRate001}
  ]
  particleSpawnScript [
   &InitializeParticle(
    Color=&MakeLinearColorFromVectorAndFloat(
     "Vector (RGB)"=&MakeVectorFromLinearColorRGB(Color=<rgba 0.4583 1 0.4955 1>)
     "Float (Alpha)"=&FN_FrontEndFloatModifier(Gameplay=0.25, Front End=0.1)
    )
    Lifetime Max=0.16
    Lifetime Min=0.1
    Lifetime Mode=::NewEnumerator1
    Sprite Size Max=<4 10>
    Sprite Size Min=<3 5>
    Sprite Size Mode=::NewEnumerator2
    UsePositionOffset=true
    Write Material Random=true
   )
   &ShapeLocation(
    Handle Radius=5
    Large Radius=&FN_FrontEndFloatModifier(Gameplay=250, Front End=200)
    MaskSpawnGroup=false
    Offset=&FN_FrontEndVector3Modifier(Gameplay=<0 -200 15>, Front End=<0 -200 15>)
    Offset Mode=::NewEnumerator0
    Override Seed=false
    Rotation Axis=<0 1 0>
    Rotation Mode=::NewEnumerator1
    Shape Primitive=::NewEnumerator3
    UseOwnerScale=false
   )
   /SetVariables(Particles.Position=&Multiply_Vector(A=Linked() {linkedVariable {name=Particles.Position, type=NiagaraPosition}}, B=<1.2 -1 0.8>)) {isSetParameters=true}
   &AddVelocityFromPoint(
    Module.Use Falloff Distance=false
    Origin Offset=<0 1000 -250>
    Velocity Strength=&FN_FrontEndFloatModifier(Gameplay=1000, Front End=0)
   )
   &AddVelocity(
    Velocity=&FN_FrontEndVector3Modifier(
     Gameplay=&Add_Vector(A=<0 300 0>, B=&Multiply_VectorByFloat(Vector=<0 0 222>, Float=$System.SmoothedForwardDot))
     Front End=<0 -100 -10>
    )
    Velocity Speed Scale=&FN_FrontEndFloatModifier(Gameplay=-1, Front End=1)
   )
  ]
  particleUpdateScript [
   &ParticleState()
   &ScaleSpriteSize(Scale Factor=&Vector2DFromCurve(
     Vector2Curve=DataInterface() {propertyValues {XCurve {keys [<at 0 0.5 RCIM_Cubic>,<at 0.5 1 RCIM_Cubic>]}, YCurve {keys [<at 0 0 RCIM_Cubic>,<at 0.5 1 RCIM_Cubic>]}}}
     ScaleCurve=&Vector2DFromFloat(Value=1)
    ))
   &SolveForcesAndVelocity(Clamp Velocity=false, Limit Acceleration=false)
   &ScaleSpriteSizeBySpeed(Max Scale Factor=<1 50>, Min Scale Factor=<0.01 0.05>, Velocity Threshold=1)
   &ScaleColor(
    Scale Alpha=&FloatFromCurve(
     FloatCurve=DataInterface() {propertyValues {Curve {keys [
         {arriveTangent=6.8397, interpMode=RCIM_Cubic, leaveTangent=6.8397, tangentMode=RCTM_User, time=0, value=0}
         <at 0.3 1 RCIM_Cubic>
         {arriveTangent=-2.221, interpMode=RCIM_Cubic, leaveTangent=-2.221, tangentMode=RCTM_User, time=1, value=0}
        ]}}}
     Scale Curve=&FN_FrontEndFloatModifier(
      Gameplay=&Lerp_Float(
       A=0
       B=&FNTimeOfDayScalar(Night Scale=0.2, Morning Scale=0.3, Day Scale=0.4, Evening Scale=0.3)
       Alpha=&Multiply_Float(A=$System.SystemFadeIn, B=$System.Timeline.Time)
      )
      Front End=&Lerp_Float(A=0, B=0.4, Alpha=$System.SystemFadeIn)
     )
    )
    Scale RGB=&VectorFromFloat(Value=&FN_FrontEndFloatModifier(Gameplay=1, Front End=1))
    ScaleA=true
    ScaleRGB=true
   )
  ]
  rendererClasses [&NiagaraSpriteRendererProperties]
  renderers [{rendererClass=&NiagaraSpriteRendererProperties}]
 }
 NE_SpiralRibbon {
  emitterSpawnScript [/SetVariables(Emitter.ParticlePerRibbon=8, Emitter.RibbonNumber=3) {isSetParameters=true}]
  emitterUpdateScript [
   &EmitterState(
    Inactive Response=::NewEnumerator1
    Life Cycle Mode=::NewEnumerator1
    Loop Behavior=::NewEnumerator1
    Loop Duration Mode=::NewEnumerator1
    UseLoopDelay=false
   )
   &SpawnBurst_Instantaneous(
    Spawn Count=&Multiply_Int(A=$Emitter.ParticlePerRibbon, B=$Emitter.RibbonNumber)
    Use Spawn Probability=false
    UseLoopCountLimit=false
   )
  ]
  particleSpawnScript [
   &InitializeParticle(
    Lifetime=5
    Ribbon Width=20
    Ribbon Width Mode=::NewEnumerator1
    Sprite Size Mode=::NewEnumerator3
    UsePositionOffset=false
    Write Material Random=true
   )
   &PartitionParticles(
    ID to Partition:NiagaraInt32=&ReturnExecIndex()
    Normalized Index In Partition Includes One=true
    Number Of Partitions=$Emitter.RibbonNumber
    Particles Per Partition:NiagaraInt32=$Emitter.ParticlePerRibbon
    Partition Mode:ENiagara_PartitionMode=::NewEnumerator1
    Write Partition Index to Ribbon Link Order:ENiagara_AttributeSamplingApplyOutput=::NewEnumerator0
    Write Partition to Ribbon ID=::NewEnumerator0
   )
   /SetVariables(Particles.NormalizedRibbonPos:NiagaraFloat=&OneMinusFloat(Float=$Particles.PartitionParticles.Normalized Index In Partition)) {isSetParameters=true, module=SetVariables_#9}
   /SetVariables(
    Particles.SpiralFreq=0.6
    Particles.SpiralStrength:NiagaraFloat=&Multiply_Float(
     A=15
     B=&FloatFromCurve(
      FloatCurve=DataInterface() {propertyValues {Curve {keys [
          {arriveTangent=3.6599, interpMode=RCIM_Cubic, leaveTangent=3.6599, tangentMode=RCTM_User, time=0, value=-0.025}
          {interpMode=RCIM_Cubic, tangentMode=RCTM_User, time=0.5, value=1}
          {arriveTangent=-0.1348, interpMode=RCIM_Cubic, leaveTangent=-0.1348, tangentMode=RCTM_User, time=1, value=0}
         ]}}}
      CurveIndex=$Particles.NormalizedRibbonPos
     )
    )
   ) {isSetParameters=true}
   /SetVariables(Particles.RibbonU0Override=$Particles.NormalizedRibbonPos) {isSetParameters=true}
   /SetVariables() {isSetParameters=true, module=SetVariables_#10}
   &SkeletalMeshLocation(
    Bone / Socket Index=$Particles.PartitionParticles.Partition
    Bone Sampling Mode=::NewEnumerator4
    MaskSpawnGroup=false
    Sampled Position Offset=<0 10 0>
    Skeletal Mesh=$System.Skeletal Mesh
    Use Sampled Mesh Attribute as Ribbon ID=false
   ) {module=SkeletalMeshLocation001}
  ]
  particleUpdateScript [
   &ParticleState(Kill Particles When Lifetime Has Elapsed=false)
   &SkeletalMeshLocation(
    Bone / Socket Index=$Particles.PartitionParticles.Partition
    Bone Sampling Mode=::NewEnumerator4
    MaskSpawnGroup=false
    Position Sampling=::NewEnumerator2
    Sampled Position Offset=$System.BaseOffset
    Skeletal Mesh=$System.Skeletal Mesh
    Use Sampled Mesh Attribute as Ribbon ID=false
   )
   /SetVariables(Particles.UVOffset:NiagaraFloat=&Add_Float(
     A=$Particles.UVOffset
     B=&Multiply_Float(A=$Engine.DeltaTime, B=&Lerp_Float(A=-0.1, B=-0.3, Alpha=$System.SmoothedForwardDot))
    )) {isSetParameters=true}
   /SetVariables(Particles.GoalPos:NiagaraPosition=&AddVectorToPosition(
     Position=$Particles.SkeletalMeshLocation.SampledPosition
     Vector=&Lerp_Vector(
      A=<0 0 0>
      B=&Lerp_Vector(
       A=<0 -20 0>
       B=&Lerp_Vector(
        A=<0 -100 20>
        B=&FN_FrontEndVector3Modifier(Gameplay=<0 -215 20>, Front End=<0 -200 -20>)
        Alpha=&FN_FrontEndFloatModifier(Gameplay=$System.SmoothedForwardDot, Front End=1)
       )
       Alpha=$System.SystemFadeIn
      )
      Alpha=$Particles.NormalizedRibbonPos
     )
    )) {isSetParameters=true, module=SetVariables_#11}
   /SetVariables(Particles.SpiralStrength:NiagaraFloat=&Multiply_Float(
     A=&Lerp_Float(A=3, B=4, Alpha=$System.SmoothedForwardDot)
     B=&FloatFromCurve(
      FloatCurve=DataInterface() {propertyValues {Curve {keys [
          {arriveTangent=0.3159, interpMode=RCIM_Cubic, leaveTangent=0.3159, tangentMode=RCTM_User, time=0, value=1.5}
          {arriveTangent=0.0033, interpMode=RCIM_Cubic, leaveTangent=0.0033, tangentMode=RCTM_User, time=0.2, value=1.5}
          {arriveTangent=-2.5078, interpMode=RCIM_Cubic, leaveTangent=-2.5078, tangentMode=RCTM_User, time=1, value=0.3}
         ]}}}
      CurveIndex=$Particles.NormalizedRibbonPos
     )
    )) {isSetParameters=true}
   /SetVariables(Particles.SpiralTime:NiagaraFloat=&Add_Float(
     A=&Add_Float(
      A=&Add_Float(A=&Multiply_Float(A=$System.Age, B=-1), B=$Particles.UVOffset)
      B=$Particles.NormalizedRibbonPos
     )
     B=&Multiply_FloatByInt(
      Float=&Divide_Float(A=$Particles.SpiralFreq, B=&MakeFloatFromInt(Int=$Emitter.RibbonNumber))
      Integer=$Particles.PartitionParticles.Partition
     )
    )) {isSetParameters=true}
   /SetVariables(Particles.SpiralOffset:Vector3f=&MakeVector(
     X=&Sine(Normalized Angle=$Particles.SpiralTime, Period=$Particles.SpiralFreq, Scale=$Particles.SpiralStrength, Bias=0)
     Y=0
     Z=&Cosine(Normalized Angle=$Particles.SpiralTime, Period=$Particles.SpiralFreq, Scale=$Particles.SpiralStrength, Bias=0)
    )) {isSetParameters=true}
   &Color(
    Color=&MultiplyLinearColor(
     Linear Color=&ColorFromCurve(
      ColorCurve=DataInterface() {propertyValues {
        AlphaCurve {keys [{arriveTangent=2.9004, interpMode=RCIM_Cubic, time=0.0541, value=1}, <at 0.3539 1>, <at 0.697 0 RCIM_Cubic>]}
        BlueCurve {keys [<at 0 2>,<at 0.1907 1>,<at 0.782 0.2>]}
        GreenCurve {keys [<at 0 2>,<at 0.1907 1>,<at 0.782 0.2>]}
        RedCurve {keys [<at 0 2>,<at 0.1907 1>,<at 0.782 0.2>]}
       }}
      CurveIndex=$Particles.NormalizedRibbonPos
     )
     Scale Factor=&MakeVector4_FromLinearColor(LinearColor=&LerpLinearColors(StartColor=<rgba 0.7 0.7 0.7 1>, EndColor=<rgba 1.5 1.5 1.5 1>, LerpFactor=$System.SmoothedForwardDot))
    )
    Scale Alpha=&FloatFromCurve(
     FloatCurve=DataInterface() {propertyValues {Curve {keys [
         {interpMode=RCIM_Cubic, leaveTangent=-20.1175, tangentMode=RCTM_Break, time=0, value=1}
         {interpMode=RCIM_Cubic, tangentMode=RCTM_User, time=0.15, value=0}
         {arriveTangent=16.0686, interpMode=RCIM_Cubic, tangentMode=RCTM_Break, time=0.4, value=1}
         {arriveTangent=-0.2517, interpMode=RCIM_Cubic, leaveTangent=-12.1505, tangentMode=RCTM_Break, time=0.5, value=1}
         {arriveTangent=0.0335, interpMode=RCIM_Cubic, leaveTangent=0.0335, tangentMode=RCTM_User, time=0.7, value=0}
         {arriveTangent=12.5546, interpMode=RCIM_Cubic, tangentMode=RCTM_Break, time=0.9, value=1}
        ]}}}
     CurveIndex=&ModuloFloat(
      A=&Abs_Float(Float:NiagaraFloat=&Add_Float(
        A=$Particles.NormalizedRibbonPos
        B=&Multiply_Float(
         A=&Add_Float(A=$Particles.Age, B=&Multiply_FloatByInt(Float=0.2351, Integer=$Particles.PartitionParticles.Partition))
         B=-0.5
        )
       ))
      B=1
     )
    )
   ) {module=Color001}
   &InheritSourceMovement(
    Applied Position Scale=&VectorFromFloat(Value=&Multiply_Float(
      A=-0.5
      B=&FloatFromCurve(
       FloatCurve=DataInterface() {propertyValues {Curve {keys [
           {arriveTangent=-0.0194, interpMode=RCIM_Cubic, leaveTangent=-0.0194, tangentMode=RCTM_User, time=0.2, value=0}
           {arriveTangent=1.6438, interpMode=RCIM_Cubic, leaveTangent=1.6438, tangentMode=RCTM_User, time=1, value=1}
          ]}}}
       CurveIndex=$Particles.NormalizedRibbonPos
      )
     ))
    ApplyForce=false
    ApplyPosition=true
    ApplyPositionFalloff=false
    Source Velocity:Vector3f=&TransformVector(Vector=$Engine.Owner.Velocity, Source Space=::World, Destination Space=::Local)
   )
   &SolveForcesAndVelocity(Clamp Velocity=false, Limit Acceleration=false)
   /SetVariables(Particles.Position=&AddVectorToPosition(
     Position=&Lerp_Vector(
      A=Linked() {linkedVariable {name=Particles.Position, type=NiagaraPosition}}
      B=Linked() {linkedVariable {name=Particles.GoalPos, type=NiagaraPosition}}
      Alpha=&Multiply_Float(A=$Engine.DeltaTime, B=10)
     )
     Vector=$Particles.SpiralOffset
    )) {isSetParameters=true, module=SetVariables_#12}
   /SetVariables(Particles.RibbonFacing=&NormalizeVector(
     Vector To Normalize=&TransformVector(Vector=$Particles.SpiralOffset, Destination Space=::Simulation)
     Safe Normalize=true
     Safe Fallback Threshold=0
     Safe Fallback Vector=<0 0 0.001>
    )) {isSetParameters=true}
   /SetVariables(Particles.BackSpiralColor:LinearColor=&MultiplyLinearColor(
     Linear Color=$Particles.Color
     Scale Factor=&MakeVector4_FromLinearColor(LinearColor=<rgba 0.1875 0.5 0.1224 1>)
    )) {isSetParameters=true}
   /SetVariables(Particles.BackRibbonFacing:Vector3f=&NormalizeVector(
     Vector To Normalize=&TransformVector(Vector=&Multiply_VectorByFloat(Vector=$Particles.SpiralOffset, Float=-1), Destination Space=::Simulation)
     Safe Normalize=true
     Safe Fallback Threshold=0
     Safe Fallback Vector=<0 0 0.001>
    )) {isSetParameters=true}
   &ScaleRibbonWidth(Ribbon Width Scale=&FloatFromCurve(
     FloatCurve=DataInterface() {propertyValues {Curve {keys [
         {arriveTangent=17.9794, interpMode=RCIM_Cubic, leaveTangent=17.9794, tangentMode=RCTM_User, time=0, value=0}
         {interpMode=RCIM_Cubic, tangentMode=RCTM_User, time=0.2, value=2}
         {arriveTangent=-1.6291, interpMode=RCIM_Cubic, leaveTangent=-1.6291, tangentMode=RCTM_User, time=1, value=1}
        ]}}}
     CurveIndex=$Particles.NormalizedRibbonPos
     Scale Curve=&Lerp_Float(A=0.6, B=0.8, Alpha=$System.SmoothedForwardDot)
    ))
   /SetVariables(Particles.GlowWidth:NiagaraFloat=&Multiply_Float(A=$Particles.RibbonWidth, B=1.5)) {isSetParameters=true}
   /SetVariables(Particles.GlowColor=&MultiplyLinearColor(Linear Color=$Particles.Color, Scale Factor=&MakeVector4_FromLinearColor(LinearColor=<rgba 1 1 1 5>))) {isSetParameters=true}
   &DynamicMaterialParameters(
    Index 1 Param 3=&FloatFromCurve(
     FloatCurve=DataInterface() {propertyValues {Curve {keys [
         {arriveTangent=0.0521, interpMode=RCIM_Cubic, leaveTangent=0.0521, tangentMode=RCTM_User, time=0, value=0}
         {interpMode=RCIM_Cubic, tangentMode=RCTM_User, time=1, value=1.5}
        ]}}}
     CurveIndex=$Particles.NormalizedRibbonPos
    )
    Module.Index0Param0Write=true
    Module.Index0Param1Write=true
    Module.Index0Param2Write=true
    Module.Index0Param3Write=true
    Module.Index1Param0Write=true
    Module.Index1Param1Write=true
    Module.Index1Param2Write=true
    Module.Index1Param3Write=true
    Write Parameter Index 1=true
   )
   &CameraOffset()
  ]
  rendererClasses [&NiagaraRibbonRendererProperties]
  renderers [
   {rendererClass=&NiagaraRibbonRendererProperties}
   {rendererClass=&NiagaraRibbonRendererProperties, rendererIndex=1}
   {rendererClass=&NiagaraRibbonRendererProperties, rendererIndex=2}
   {rendererClass=&NiagaraRibbonRendererProperties, rendererIndex=3}
   {rendererClass=&NiagaraRibbonRendererProperties, rendererIndex=4}
  ]
 }
 NE_StraightRibbon {
  emitterSpawnScript [/SetVariables(Emitter.ParticlePerRibbon=8) {isSetParameters=true}]
  emitterUpdateScript [
   &EmitterState(
    Inactive Response=::NewEnumerator1
    Life Cycle Mode=::NewEnumerator1
    Loop Behavior=::NewEnumerator1
    Loop Duration Mode=::NewEnumerator1
    UseLoopDelay=false
   )
   &SpawnBurst_Instantaneous(
    Spawn Count=&Multiply_Int(A=3, B=$Emitter.ParticlePerRibbon)
    Use Spawn Probability=false
    UseLoopCountLimit=false
   )
  ]
  particleSpawnScript [
   &InitializeParticle(Lifetime=5, Ribbon Width=30, Ribbon Width Mode=::NewEnumerator1, UsePositionOffset=false, Write Material Random=true)
   /SetVariables(Particles.VisibilityTag=&SpawnGroup()) {isSetParameters=true}
   &PartitionParticles(
    ID to Partition:NiagaraInt32=&ReturnExecIndex()
    Normalized Index In Partition Includes One=true
    Number Of Partitions=3
    Particles Per Partition:NiagaraInt32=$Emitter.ParticlePerRibbon
    Partition Mode:ENiagara_PartitionMode=::NewEnumerator1
    Write Partition Index to Ribbon Link Order:ENiagara_AttributeSamplingApplyOutput=::NewEnumerator0
    Write Partition to Ribbon ID=::NewEnumerator0
   )
   /SetVariables(Particles.NormalizedRibbonPos:NiagaraFloat=&OneMinusFloat(Float=$Particles.PartitionParticles.Normalized Index In Partition)) {isSetParameters=true, module=SetVariables_#9}
   /SetVariables(Particles.RibbonLinkOrder:NiagaraFloat=$Particles.NormalizedRibbonPos) {isSetParameters=true}
   /SetVariables(Particles.GlowColor=&ColorFromCurve(
     ColorCurve=DataInterface() {propertyValues {
       AlphaCurve {keys [<at 0 1 RCIM_Cubic>,<at 0.52 1>,<at 1 0>]}
       BlueCurve {keys [<at 0 1 RCIM_Cubic>]}
       GreenCurve {keys [<at 0 1 RCIM_Cubic>]}
       RedCurve {keys [<at 0 1 RCIM_Cubic>]}
      }}
     CurveIndex=$Particles.NormalizedRibbonPos
    )) {isSetParameters=true}
   /SetVariables(Particles.UVOffset:NiagaraFloat=&Multiply_FloatByInt(Float=0.33, Integer=$Particles.PartitionParticles.Partition)) {isSetParameters=true, module=SetVariables_#10}
   &SkeletalMeshLocation(
    Bone / Socket Index=$Particles.PartitionParticles.Partition
    Bone Sampling Mode=::NewEnumerator4
    MaskSpawnGroup=false
    Sampled Position Offset=<0 10 0>
    Skeletal Mesh=$System.Skeletal Mesh
    Use Sampled Mesh Attribute as Ribbon ID=false
   ) {module=SkeletalMeshLocation001}
  ]
  particleUpdateScript [
   &ParticleState(Kill Particles When Lifetime Has Elapsed=false)
   &SkeletalMeshLocation(
    Bone / Socket Index=$Particles.PartitionParticles.Partition
    Bone Sampling Mode=::NewEnumerator4
    MaskSpawnGroup=false
    Position Sampling=::NewEnumerator2
    Sampled Position Offset=$System.BaseOffset
    Skeletal Mesh=$System.Skeletal Mesh
    Use Sampled Mesh Attribute as Ribbon ID=false
   )
   &SkeletalMeshLocation(
    Bone Sampling Mode=::NewEnumerator4
    MaskSpawnGroup=false
    Position Sampling=::NewEnumerator2
    Sampled Position Offset=$System.BaseOffset
    Skeletal Mesh=$System.Skeletal Mesh
    Use Sampled Mesh Attribute as Ribbon ID=false
   ) {module=SkeletalMeshLocation002}
   /SetVariables(Particles.UVOffset:NiagaraFloat=&Add_Float(
     A=$Particles.UVOffset
     B=&Multiply_Float(A=$Engine.DeltaTime, B=&Lerp_Float(A=-0.1, B=-0.3, Alpha=$System.SmoothedForwardDot))
    )) {isSetParameters=true}
   &Color(
    Color=&ColorFromCurve(
     ColorCurve=DataInterface() {propertyValues {
       AlphaCurve {keys [{arriveTangent=2.4905, interpMode=RCIM_Cubic, time=0, value=1}]}
       BlueCurve {keys [<at 0 1>]}
       GreenCurve {keys [<at 0 1>]}
       RedCurve {keys [<at 0 1>]}
      }}
     CurveIndex=$Particles.NormalizedRibbonPos
    )
    Scale Alpha=&Multiply_Float(
     A=&FloatFromCurve(
      FloatCurve=DataInterface() {propertyValues {Curve {keys [
          {arriveTangent=-0.3141, interpMode=RCIM_Cubic, leaveTangent=-0.3141, tangentMode=RCTM_User, time=0.2, value=1}
          {arriveTangent=0.0794, interpMode=RCIM_Cubic, leaveTangent=0.0794, tangentMode=RCTM_User, time=1, value=0}
         ]}}}
      CurveIndex=$Particles.NormalizedRibbonPos
     )
     B=&FloatFromCurve(
      FloatCurve=DataInterface() {propertyValues {Curve {keys [<at 0 1>,<at 0.05 1>]}}}
      CurveIndex=&OneMinusFloat(Float=&Abs_Float(Float:NiagaraFloat=&DotProduct(Vector A=<0 1 0>, Vector B=&ReturnCameraVector(Camera Vectors=::NewEnumerator1, Coordinate Space=::Simulation))))
     )
    )
    Scale Color=&VectorFromFloat(Value=&Lerp_Float(A=1, B=2, Alpha=$System.SmoothedForwardDot))
   )
   /SetVariables(Particles.GoalPos:NiagaraPosition=&AddVectorToPosition(
     Position=$Particles.SkeletalMeshLocation.SampledPosition
     Vector=&Lerp_Vector(
      A=<0 0 0>
      B=&Lerp_Vector(
       A=<0 0 0>
       B=&Lerp_Vector(
        A=<0 -90 0>
        B=&FN_FrontEndVector3Modifier(Gameplay=<0 -215 20>, Front End=<0 -200 -20>)
        Alpha=&FN_FrontEndFloatModifier(Gameplay=$System.SmoothedForwardDot, Front End=1)
       )
       Alpha=$System.SystemFadeIn
      )
      Alpha=$Particles.NormalizedRibbonPos
     )
    )) {isSetParameters=true, module=SetVariables_#11}
   &InheritSourceMovement(
    Applied Position Scale=&VectorFromFloat(Value=&Multiply_Float(
      A=-0.5
      B=&FloatFromCurve(
       FloatCurve=DataInterface() {propertyValues {Curve {keys [
           {arriveTangent=-0.0194, interpMode=RCIM_Cubic, leaveTangent=-0.0194, tangentMode=RCTM_User, time=0.2, value=0}
           {arriveTangent=1.6438, interpMode=RCIM_Cubic, leaveTangent=1.6438, tangentMode=RCTM_User, time=1, value=1}
          ]}}}
       CurveIndex=$Particles.NormalizedRibbonPos
      )
     ))
    ApplyForce=false
    ApplyPosition=true
    ApplyPositionFalloff=false
    Source Velocity:Vector3f=&TransformVector(Vector=$Engine.Owner.Velocity, Source Space=::World, Destination Space=::Local)
   )
   &SolveForcesAndVelocity(Clamp Velocity=false, Limit Acceleration=false)
   /SetVariables(Particles.Position=&Lerp_Vector(
     A=Linked() {linkedVariable {name=Particles.Position, type=NiagaraPosition}}
     B=Linked() {linkedVariable {name=Particles.GoalPos, type=NiagaraPosition}}
     Alpha=&Multiply_Float(A=$Engine.DeltaTime, B=10)
    )) {isSetParameters=true, module=SetVariables_#12}
   &ScaleRibbonWidth(Ribbon Width Scale=&Multiply_Float(
     A=&FloatFromCurve(
      FloatCurve=DataInterface() {propertyValues {Curve {keys [
          {arriveTangent=-0.0253, interpMode=RCIM_Cubic, leaveTangent=-0.0253, tangentMode=RCTM_User, time=0, value=0.8}
          {arriveTangent=0.0106, interpMode=RCIM_Cubic, leaveTangent=0.0106, tangentMode=RCTM_User, time=0.1, value=1}
          {arriveTangent=-0.0039, interpMode=RCIM_Cubic, leaveTangent=-0.0039, tangentMode=RCTM_User, time=1, value=0}
         ]}}}
      CurveIndex=$Particles.NormalizedRibbonPos
      Scale Curve=$System.SystemFadeIn
     )
     B=&Lerp_Float(A=0.8, B=1.2, Alpha=$System.SmoothedForwardDot)
    ))
   /SetVariables(Particles.GlowWidth:NiagaraFloat=&Multiply_Float(A=$Particles.RibbonWidth, B=3)) {isSetParameters=true}
   &DynamicMaterialParameters(
    Index 0 Param 1=$Particles.UVOffset
    Index 0 Param 2=&FN_FrontEndFloatModifier(Gameplay=$System.SmoothedForwardDot, Front End=1)
    Module.Index0Param0Write=true
    Module.Index0Param1Write=true
    Module.Index0Param2Write=true
    Module.Index0Param3Write=true
   )
   /SetVariables(Particles.GlowColor=&MultiplyLinearColor(
     Linear Color=$Particles.Initial.GlowColor
     Scale Factor=&MakeVector4_FromLinearColor(LinearColor=$System.StraightColorMid)
    )) {isSetParameters=true}
  ]
  rendererClasses [&NiagaraRibbonRendererProperties]
  renderers [
   {rendererClass=&NiagaraRibbonRendererProperties}
   {rendererClass=&NiagaraRibbonRendererProperties, rendererIndex=1}
  ]
 }
}
summary {
 systemName=NS_Glider_AnglePatch_Trail
 emitters [
  {emitterName=NE_StraightRibbon, rendererClasses [&NiagaraRibbonRendererProperties]}
  {
   emitterName=NE_JetGlow
   rendererClasses [
    &NiagaraSpriteRendererProperties
    &NiagaraLightRendererProperties
    &NiagaraMeshRendererProperties
   ]
  }
  {emitterName=NE_Speed_Lines, rendererClasses [&NiagaraSpriteRendererProperties]}
  {emitterName=NE_SpiralRibbon, rendererClasses [&NiagaraRibbonRendererProperties]}
  {emitterName=NE_GooDrops001, rendererClasses [&NiagaraSpriteRendererProperties]}
 ]
 userVariables [
  User.bIsFrontEnd {type=NiagaraBool}
  User.bIsFrontEndPreview {type=NiagaraBool}
  User.bIsFullyDeployed {type=NiagaraBool}
  User.ForwardDot {type=NiagaraFloat, defaultValue=0, description=""}
  User.VariantPicker {type=NiagaraFloat, defaultValue=0, description=""}
 ]
}
