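# Masked, two-sided material for the VinderTech Angle Patch glider, with a deploy effect that reveals it from hidden through glowing wires to its finished surface.
# Custom primitive data timestamps drive an eased timeline, offset along a pre-skinned local-position gradient to sweep both transitions.
# Texture-masked, temporally dithered wires blend with the character-parent surface, which adds clear coat and metal roughness control before the PostFX function.
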Material /BRCosmetics/Gadgets/Assets/VinderTech_GliderChute/Glider_AnglePatch/Materials/M_Glider_AnglePatch
graph {
 properties {
  bUseMaterialAttributes=true
  bUsedWithMeshDeformer=true
  bUsedWithSkeletalMesh=true
  blendMode=BLEND_Masked
  pixelDepthOffsetMode=PDOM_Legacy
  shadingModel=MSM_FromMaterialExpression
  twoSided=true
  usedShadingModels=Default Lit | Subsurface | Clear Coat | Subsurface Profile | Cloth
 }
 nodes [
  n1 = MaterialFunctionCall() {materialFunction=&MF_CharacterParent_2023_PreFX}
  n2 = GetMaterialAttributes(None=@n1.Result) {attributeGetTypes [#1, #2]}
  n3 = Multiply(A=@n2.Roughness, B=ScalarParameter() {defaultValue=1, parameterName=Metal Roughness Multiplier})
  n4 = Saturate(None=Power(Base=@n2.Metallic))
  n5 = MaterialFunctionCall!(
   Mask=@n1.Tech Art Mask
   Fresnel=@n1.Tech Art Mask
   In=SetMaterialAttributes!(MaterialAttributes=@n2.MaterialAttributes, Roughness=LinearInterpolate(A=@n2.Roughness, B=@n3, Alpha=@n4)) {attributeSetTypes [#1]}
  ) {materialFunction=/Game/Global/MaterialFunctions/TechArt/MF_Advanced_CarClearCoat}
  ClearCoatNormalCustomOutput!(None=MaterialFunctionCall(BaseNormal=@n5.NormalWithOrangePeel, AdditionalNormal=@n5.NormalWithOrangePeel) {materialFunction=&BlendAngleCorrectedNormals}.Result)
  Frac(None=Divide(A=Time()))
  LinearInterpolate(
   A=LinearInterpolate(
    A=Constant3Vector() {desc=hidden, constant=<rgba 1 0 0 0>}
    B=Constant3Vector() {desc=Wires, constant=<rgba 0 1 0 0>}
    Alpha=NamedRerouteUsage()
   )
   B=Constant3Vector() {desc=PBR, constant=<rgba 0 0 1 0>}
   Alpha=NamedRerouteUsage()
  )
  Multiply(A=ScalarParameter() {defaultValue=1, parameterName=TL_Parent, sliderMax=1}) {constB=2}
  n6 = LocalPosition() {localOrigin=InstancePreSkinning}
  n7 = Divide(A=OneMinus(None=ComponentMask(None=@n6.XYZ) {g=true})) {constB=512}
  n8 = ShaderStageSwitch(PixelShader=VertexInterpolator(VS=@n7).PS, VertexShader=@n7) {desc=glider gradient}
  n9 = NamedRerouteUsage()
  n10 = ScalarParameter() {defaultValue=0.05, parameterName=WireToOriginalDelay, sliderMax=1}
  n11 = Add(A=Subtract(A=@n8) {constB=0.5}, B=Saturate(None=Subtract(A=@n9, B=@n10))) {desc=shader swap time line}
  n12 = SphereMask(A=Add(A=@n11) {constB=0.5}, B=Constant() {r=0.5}) {attenuationRadius=0.1, hardnessPercent=0}
  n13 = TextureSampleParameter2D() {
   parameterName=Masks
   samplerType=SAMPLERTYPE_Masks
   texture=/BRCosmetics/Gadgets/Assets/VinderTech_GliderChute/Glider_AnglePatch/Textures/T_Glider_AnglePatch_FX2
  }
  n14 = Saturate(None=Add(A=@n13.B, B=ScalarParameter() {parameterName=WireMaskBias}))
  n15 = ScalarParameter() {defaultValue=2.2, parameterName=WireMaskPower}
  n16 = Add(A=@n12, B=Multiply(A=ScalarParameter() {defaultValue=1, parameterName=WireMaskStrength}, B=Power(Base=@n14, Exp=@n15) {constExponent=2.2}))
  n17 = VectorParameter() {parameterName=GlowColor, defaultValue=<rgba 0 1 0 10>}
  NamedRerouteDeclaration(None=SetMaterialAttributes!(
    MaterialAttributes
    Opacity Mask=MaterialFunctionCall(Alpha Threshold=@n16) {materialFunction=&DitherTemporalAA}.Result
    Emissive Color=Multiply(A=@n17.RGB, B=@n17.RGB)
   ) {attributeSetTypes [#3, #4]}) {name=MA_Wire, nodeColor=<rgba 0 1 0.205 1>}
  NamedRerouteDeclaration(None=@n5.Result) {name=MA_Original, nodeColor=<rgba 0.0069 0 1 1>}
  n18 = NamedRerouteUsage()
  n19 = NamedRerouteUsage()
  n20 = NamedRerouteUsage()
  NamedRerouteDeclaration(None=BlendMaterialAttributes(A=BlendMaterialAttributes(A=@n18, B=@n19, Alpha=@n20), B=NamedRerouteUsage(), Alpha=NamedRerouteUsage())) {name=MA_Final, nodeColor=<rgba 1 0 0.9844 1>}
  NamedRerouteDeclaration(None=SetMaterialAttributes!(MaterialAttributes, Opacity Mask=Constant(), Shading Model=ShadingModel()) {attributeSetTypes [#3, #5]}) {name=MA_Hidden, nodeColor=<rgba 1 0.0102 0 1>}
  n21 = NamedRerouteUsage()
  n22 = ScalarParameter() {defaultValue=1.15, parameterName=HiddenToWireSpeed}
  n23 = Add(A=Subtract(A=@n8) {constB=0.5}, B=Saturate(None=Multiply(A=@n21, B=@n22))) {desc=shader swap time line}
  n24 = Add(A=@n23) {constB=0.5}
  n25 = Constant() {r=0.5}
  NamedRerouteDeclaration(None=Subtract(A=Ceil(None=@n23), B=SphereMask(A=@n24, B=@n25) {attenuationRadius=0.1, hardnessPercent=0})) {name=Blend_Hidden2Wire, nodeColor=<rgba 1 0.375 0 1>}
  NamedRerouteDeclaration(None=Subtract(A=Ceil(None=@n11), B=@n12)) {name=Blend_Wire2Original, nodeColor=<rgba 0.0547 0 1 1>}
  n26 = ScalarParameter() {bUseCustomPrimitiveData=true, parameterName=DeployTimeStamp_CPD, primitiveDataIndex=1}
  n27 = Subtract(A=Time(), B=@n26)
  n28 = Divide(
   A=Subtract(A=Abs(None=@n27), B=ScalarParameter() {defaultValue=0.1, parameterName=DeployAnimDelay})
   B=ScalarParameter() {defaultValue=1, parameterName=DeployAnimSpeed}
  )
  n29 = LinearInterpolate(
   A=Constant() {r=1}
   B=Multiply(A=Saturate(None=@n28), B=Saturate(None=@n26))
   Alpha=ScalarParameter() {bUseCustomPrimitiveData=true, parameterName=DeployTimeStampIsSet_CPD, primitiveDataIndex=2}
  )
  NamedRerouteDeclaration(None=SquareRoot(None=NamedRerouteDeclaration(None=@n29) {name=Time, nodeColor=<rgba 1 0.2109 0 1>})) {name=SquaredTime, nodeColor=<rgba 0 0.1094 1 1>}
  NamedRerouteUsage()
  ScalarParameter() {defaultValue=1, desc=TL_Glow, parameterName=MaterialCurveB, sliderMax=1}
  ScalarParameter() {desc=TL_WireReveal, parameterName=MaterialCurveA, sliderMax=1}
  ScalarParameter() {desc=TL_WireReveal, parameterName=MaterialCurveA, sliderMax=1}
  SetMaterialAttributes!(
   MaterialAttributes=MaterialFunctionCall(Material=NamedRerouteUsage(), M=@n1.M, Pre Skinned Local Position=@n1.M, Distance Blend=@n1.M, Fresnel=@n1.M) {materialFunction=&MF_CharacterParent_2023_PostFX}.Result
   Normal=GetMaterialAttributes(None=@n1.Result) {attributeGetTypes [#6]}.Normal
  ) {attributeSetTypes [#6]}
 ]
}
